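# Material function for gem materials that computes a distance-corrected cubemap lookup direction by intersecting a view-derived ray with a proxy sphere, without sampling a texture.
# A static switch selects IOR-based refraction or reflection, using a custom or vertex normal.
# The quadratic's far root on a Radius-scaled sphere around CubePosition (actor position by default) yields the corrected ray.
# It also outputs both raw rays and a hit mask from the determinant.
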MaterialFunction /Game/Global/Materials/MaterialGym/Materials/Gem/MaterialFunctions/MF_DistanceCorrectedCubemapSample
graph {nodes [
  n1 = FunctionInput() {bUsePreviewValueAsDefault=true, inputName=Radius, inputType=FunctionInput_Scalar, sortPriority=3, previewValue=<50 0 0 1>}
  n2 = CameraVectorWS()
  n3 = VertexNormalWS()
  n4 = Multiply(A=NamedRerouteDeclaration(None=FunctionInput(Preview=@n3) {bUsePreviewValueAsDefault=true, inputName=CustomNormalWS, sortPriority=6}) {name=NormalWS, nodeColor=<rgba 1 0 0.2812 1>}) {constB=-1}
  n5 = Constant() {r=1}
  n6 = FunctionInput() {bUsePreviewValueAsDefault=true, inputName=IOR, inputType=FunctionInput_Scalar, sortPriority=4, previewValue=<1.25 0 0 1>}
  n7 = NamedRerouteDeclaration(None=Normalize(VectorInput=MaterialFunctionCall!(Ray Direction=@n2, Surface Normal=@n4, Refractive Index Origin=@n5, Refractive Index Target=@n6) {materialFunction=/Engine/Functions/Engine_MaterialFunctions01/Vectors/Refract}.Vector)) {name=RefractedRay, nodeColor=<rgba 0.3359 0 1 1>}
  n8 = ReflectionVectorWS(CustomWorldNormal=NamedRerouteUsage())
  n9 = Multiply(
   A=@n1
   B=StaticSwitch(
    True=Multiply(A=@n7) {constB=-1}
    False=NamedRerouteDeclaration(None=@n8) {name=ReflectedRay, nodeColor=<rgba 0.6641 0 1 1>}
    Value=NamedRerouteUsage()
   )
  ) {desc=RayDirection}
  n10 = StaticSwitch(True=CameraPositionWS(), False=WorldPosition().XYZ, Value=NamedRerouteUsage())
  n11 = FunctionInput(Preview=ActorPositionWS()) {bUsePreviewValueAsDefault=true, inputName=CubePosition}
  n12 = NamedRerouteDeclaration(None=Subtract(
    A=NamedRerouteDeclaration(None=@n10) {name=RayOrigin, nodeColor=<rgba 1 0 0.5625 1>}
    B=NamedRerouteDeclaration(None=@n11) {name=CubePosition, nodeColor=<rgba 0 1 0.9766 1>}
   ) {desc=ReceiverToSphereCenter}) {name=X, nodeColor=<rgba 0.7812 1 0 1>}
  n13 = Multiply(A=DotProduct(A=@n9, B=@n12)) {constB=2, desc=QuadraticCoef.y}
  n14 = Multiply(A=@n1) {constB=1.2, desc=ProjectionSphereRadius}
  n15 = Subtract(
   A=NamedRerouteDeclaration(None=DotProduct(A=@n12, B=@n12) {desc=ReceiverToSphereCenterSq}) {name=XLengthSquared, nodeColor=<rgba 1 0.3516 0 1>}
   B=NamedRerouteDeclaration(None=Multiply(A=@n14, B=@n14) {constB=1.2, desc=SphereRadiusSquared}) {name=RadiusSquared, nodeColor=<rgba 0 1 0.625 1>}
  ) {desc=QuadraticCoef.z}
  n16 = NamedRerouteDeclaration(None=DotProduct(A=@n9, B=@n9) {desc=QuadraticCoef.x}) {name=RadiusSquared_1, nodeColor=<rgba 0 0.0625 1 1>}
  n17 = Subtract(A=Multiply(A=@n13, B=@n13), B=Multiply(A=Multiply(A=@n15, B=@n16)) {constB=4}) {desc=Determinant}
  n18 = SquareRoot(None=@n17)
  n19 = Add(A=@n16, B=Constant() {desc=Epsilon, r=0})
  n20 = Multiply(A=Multiply(A=Subtract(A=@n18, B=@n13), B=Divide(B=@n19) {constA=0.5, desc=InvTwoA}) {desc=FarIntersection}, B=@n9)
  n21 = NamedRerouteUsage()
  FunctionOutput(None=Subtract(A=Add(A=@n20, B=@n21) {desc=IntersectPosition}, B=NamedRerouteUsage())) {outputName=DistanceCorrectedRay}
  FunctionOutput(None=NamedRerouteUsage()) {outputName=RefractedRay, sortPriority=3}
  FunctionOutput(None=NamedRerouteUsage()) {outputName=ReflectedRay, sortPriority=2}
  FunctionOutput(None=Clamp(None=Ceil(None=@n17))) {outputName=Mask, sortPriority=1}
  NamedRerouteDeclaration(None=FunctionInput(Preview=StaticBool()) {bUsePreviewValueAsDefault=true, inputName=UseRefractedRay, inputType=FunctionInput_StaticBool, sortPriority=3}) {name=UseRefraction, nodeColor=<rgba 0.8516 1 0 1>}
 ]}
